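# Unlit translucent Niagara sprite material for flickering sparkles in the ZebraScramble Bacon skydiving contrail effect.
# Time scaled by Flicker_Speed and a per-particle random value runs through Sine and Frac, so particles flicker at different rates.
# That value blends particle color toward a Flicker_Brightness-boosted copy.
# A grayscale sparkle mask multiplies the result, which feeds MF_FX_EmissiveAdjustments as emissive.
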Material /BRCosmetics/Athena/Effects/Particles/Skydiving/NS_Contrail_ZebraScramble_Bacon/M_Sparkles
graph {
 properties {bUsedWithNiagaraSprites=true, blendMode=BLEND_Translucent, pixelDepthOffsetMode=PDOM_Legacy, shadingModel=MSM_Unlit, usedShadingModels=Unlit}
 nodes [
  n1 = TextureSample() {
   samplerType=SAMPLERTYPE_LinearGrayscale
   texture=&/T_Sparkles_Mask
  }
  Add(A=@n1.RGB, B=@n1.RGB)
  n2 = ParticleColor()
  n3 = Multiply(A=@n2.RGB, B=ScalarParameter() {defaultValue=4, parameterName=Flicker_Brightness}) {constB=2}
  n4 = Multiply(A=ScalarParameter() {defaultValue=1, parameterName=Flicker_Speed}, B=Time())
  n5 = ParticleRandom()
  n6 = Frac(None=Sine(None=Multiply(A=@n4, B=@n5)))
  MaterialFunctionCall(Emissive=Multiply(A=LinearInterpolate(A=@n2.RGB, B=@n3, Alpha=@n6), B=@n1.RGB)) {materialFunction=&MF_FX_EmissiveAdjustments}
 ]
}
