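# Unlit translucent material for Niagara mesh particles that draws wobbly toon lines in the ZebraScramble skydiving contrail.
# A gradient mask moves along UV V with particle relative time, tapering both line width and opacity of the Color parameter.
# Vertex offset widens the spline mesh perpendicular to the camera, using width from UV1.x.
# It adds two layered sine/cosine wobbles that grow along instance Z, driven by dynamic time, speed and pulse parameters.
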Material /BRCosmetics/Athena/Effects/Particles/Skydiving/NS_Contrail_ZebraScramble_Bacon/M_ToonSpline_ZebraScramble
graph {
 properties {
  bUsedWithNiagaraMeshParticles=true
  blendMode=BLEND_Translucent
  pixelDepthOffsetMode=PDOM_Legacy
  shadingModel=MSM_Unlit
  twoSided=true
  usedShadingModels=Unlit
 }
 nodes [
  n1 = CameraPositionWS()
  n2 = WorldPosition()
  n3 = Constant3Vector() {constant=<rgba 0 1 0 0>}
  n4 = CrossProduct(A=Normalize(VectorInput=Subtract(A=@n1, B=@n2.XYZ)), B=Normalize(VectorInput=Transform(None=@n3)))
  n5 = TextureCoordinate()
  n6 = Subtract(A=ComponentMask(None=@n5) {r=true}) {constB=0.5}
  n7 = ComponentMask(None=@n5) {g=true}
  n8 = Power(Base=MaterialFunctionCall!(
    Gradient=OneMinus(None=@n7)
    Mask Position=ParticleRelativeTime()
    Mask Size=DynamicParameter() {defaultValue=<rgba 0 0 1 1>, paramNames [Time Offset, Speed, Pulse, Mask_Size]}.Mask_Size
   ) {materialFunction=/Game/Effects/Fort_Effects/MaterialFunctions/MF_GradientMask}.Sinusoidal) {constExponent=4}
  n9 = Multiply(
   A=ComponentMask(None=TextureCoordinate() {coordinateIndex=1}) {r=true}
   B=Multiply(A=@n8, B=ScalarParameter() {defaultValue=2, parameterName=Width}) {constA=1}
  ) {desc="Width information\r\nin UV1.x"}
  n10 = TransformPosition(None=WorldPosition().XYZ) {transformSourceType=TRANSFORMPOSSOURCE_World, transformType=TRANSFORMPOSSOURCE_Instance}
  n11 = Saturate(None=Divide(A=ComponentMask(None=@n10) {b=true}) {constB=200})
  n12 = ScalarParameter() {defaultValue=4, parameterName=WPO Frequency}
  n13 = Multiply(A=@n12) {constB=0.333}
  n14 = Power(Base=@n11) {constExponent=4}
  n15 = DynamicParameter() {defaultValue=<rgba 0 0 1 1>, paramNames [Time Offset, Speed, Pulse, Mask_Size]}
  n16 = Add(A=@n15.Time Offset, B=Multiply(A=@n15.Speed, B=ScalarParameter() {defaultValue=8, parameterName=WPO Speed}))
  n17 = Subtract(A=Multiply(A=LinearInterpolate(A=@n13, B=@n12, Alpha=@n14) {constA=1}, B=@n11), B=@n16)
  n18 = AppendVector(A=AppendVector(A=Sine(None=@n17), B=Cosine(None=@n17)), B=Multiply(A=Sine(None=@n17) {period=3}) {constB=0.5})
  n19 = ScalarParameter() {defaultValue=2, parameterName=WPO Amplitude}
  n20 = LinearInterpolate(A=Multiply(A=@n19) {constB=0.333}, B=@n19, Alpha=@n14) {constA=2, constB=3}
  n21 = Subtract(A=Multiply(A=@n11) {constB=0.4}, B=Multiply(A=@n16) {constB=0.2})
  n22 = AppendVector(A=AppendVector(A=Sine(None=@n21), B=Cosine(None=@n21)), B=Multiply(A=Sine(None=@n21) {period=2}) {constB=0.2})
  n23 = LinearInterpolate(Alpha=@n15.Pulse) {constA=16, constB=24}
  n24 = Multiply(A=Add(A=Multiply(A=@n18, B=@n20) {constB=3}, B=Multiply(A=@n22, B=@n23) {constB=16})) {constB=0.5, desc=Reduce WPO noise on toonlines}
  Add(
   A=Multiply(A=Normalize(VectorInput=@n4), B=Multiply(A=@n6, B=@n9))
   B=Transform(None=Multiply(A=@n11, B=@n24)) {transformSourceType=TRANSFORMSOURCE_Instance}
  )
  n25 = VectorParameter() {parameterName=Color, defaultValue=<rgba 0 0 0 1>}
  MaterialFunctionCall(Emissive=@n25.RGB) {materialFunction=&MF_FX_EmissiveAdjustments}
  Multiply(A=Saturate(None=@n8), B=@n25.A)
  ScalarParameter() {defaultValue=1, parameterName=Mask Size}
 ]
}
