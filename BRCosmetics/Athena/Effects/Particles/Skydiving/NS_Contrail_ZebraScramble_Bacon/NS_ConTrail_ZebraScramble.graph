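# Niagara skydiving contrail cosmetic (ZebraScramble) that trails effects from bones or sockets on the player's skeletal mesh.
# Contrail_Mesh bursts two toon-shaded trail meshes oriented against owner velocity, with a distance-scalability timeline scaling their length.
# Sparkles adds twinkling sprites with curl noise, and Wind_Lines spawns short-lived streak meshes.
# ParaGlideLeanAlpha drives speed, offsets and material parameters, while front-end modifiers swap values for menu preview.
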NiagaraSystem /BRCosmetics/Athena/Effects/Particles/Skydiving/NS_Contrail_ZebraScramble_Bacon/NS_ConTrail_ZebraScramble
dependencies {
 usedDataInterfaces [
  &NiagaraDataInterfaceSpriteRendererInfo
  &NiagaraDataInterfaceCurve
  &NiagaraDataInterfacePlatformSet
  &NiagaraDataInterfaceArrayFloat3
  &NiagaraDataInterfaceArrayFloat
  &NiagaraDataInterfaceSkeletalMesh
 ]
 usedDynamicInputs []
 usedModules [
  &self:SystemSpawnScript.SystemScriptSource.SystemScriptGraph.NiagaraNodeAssignment_0.SetVariables_#1
  &SystemState
 ]
}
emitters {
 Contrail_Mesh {
  emitterSpawnScript [
   /Game/Effects/Niagara/MasterMaterials/ToonShader/Template/NM_ToonShaderParameters()
   /Game/Effects/Niagara/MasterMaterials/ToonShader/Template/NM_OutlineParameters()
  ]
  emitterUpdateScript [
   &EmitterState(
    Inactive Response=::NewEnumerator1
    Life Cycle Mode=::NewEnumerator1
    Loop Behavior=::NewEnumerator1
    Loop Duration Mode=::NewEnumerator1
    UseLoopDelay=false
   )
   /SetVariables(Emitter.DistanceFractionThreshold:NiagaraBool=&SetBoolByFloatComparison(A=$Emitter.Scalability.DistanceFraction, B=0.9, Comparison Type=::NewEnumerator6)) {isSetParameters=true}
   &SpawnBurst_Instantaneous(Spawn Count=2, Use Spawn Probability=false, UseLoopCountLimit=false)
   &Timeline(Play=$Emitter.DistanceFractionThreshold, Play Rate=4, Rewind Rate=4) {module=Timeline001}
  ]
  particleSpawnScript [
   &PartitionParticles()
   &InitializeParticle(
    Color Mode=::NewEnumerator0
    Mesh Scale Mode=::NewEnumerator3
    Mesh Uniform Scale=&FN_FrontEndFloatModifier(Gameplay=1, Front End=0.9)
    UsePositionOffset=false
    Write Material Random=true
   )
   &InitialMeshOrientation(
    Mesh Axis To Orient=&FN_FrontEndVector3Modifier(Gameplay=<1 0 0>, Front End=<-1 0 0>)
    Mesh Orientation Mode=::NewEnumerator2
    Orientation Coordinate Space=::Local
    Rotation=&FN_FrontEndVector3Modifier(
     Gameplay=&SelectVectorFromArray(
      Direct Array Index=$Particles.PartitionParticles.Partition
      Vector Selection Array=DataInterface() {propertyValues {FloatData [<0 0.02 0>, <0 -0.02 0>]}}
     )
     Front End=&SelectVectorFromArray(
      Direct Array Index=$Particles.PartitionParticles.Partition
      Vector Selection Array=DataInterface() {propertyValues {FloatData [<0 0.025 0>, <0 -0.025 0>]}}
     )
    )
    Rotation Coordinate Space=::NewEnumerator2
    Use Rotation Vector=true
   )
  ]
  particleUpdateScript [
   &ParticleState(Kill Particles When Lifetime Has Elapsed=false, Let Infinitely Lived Particles Die When Emitter Deactivates=true)
   &Timeline(
    Play Rate=&Lerp_Float(A=1, B=2, Alpha=&ClampFloat(Float=$User.ParaGlideLeanAlpha))
    Timeline End=10000000
   )
   &ScaleMeshSize(Scale Factor=&MakeVector(
     X=1
     Y=1
     Z=&FloatFromCurve(
      FloatCurve=DataInterface() {propertyValues {Curve {keys [
          {arriveTangent=15.9072, interpMode=RCIM_Cubic, leaveTangent=15.9072, tangentMode=RCTM_User, time=0, value=0}
          {arriveTangent=1.2036, interpMode=RCIM_Cubic, leaveTangent=1.2036, tangentMode=RCTM_User, time=0.45, value=1}
          <at 1 1>
         ]}}}
      CurveIndex=$Emitter.Age
      Scale Curve=$Emitter.Timeline001.NormalizedTime
     )
    ))
   &Color(Color=&FN_Select_Color_from_Time_Of_Day(Morning=<rgba 0.7 0.6475 0.595 1>, Day=<rgba 1 1 1 1>, Evening=<rgba 0.8 0.74 0.68 1>, Night=<rgba 0.3 0.3 0.5 1>))
   &DynamicMaterialParameters(
    Index 0 Param 1=&SelectFloatFromArray(Direct Array Index=$Particles.PartitionParticles.Partition, Float Selection Array=DataInterface() {propertyValues {FloatData [0,314]}})
    Index 0 Param 2=$Particles.Timeline.Time
    Index 0 Param 3=$User.ParaGlideLeanAlpha
    Module.Index0Param0Write=true
    Module.Index0Param1Write=true
    Module.Index0Param2Write=true
    Module.Index0Param3Write=false
   ) {module=DynamicMaterialParameters001}
   &SkeletalMeshLocation(
    Bone / Socket Index=$Particles.PartitionParticles.Partition
    Bone Sampling Mode=::NewEnumerator4
    MaskSpawnGroup=false
    Orientation Sampling=::NewEnumerator1
    Sampled Position Offset=&FN_FrontEndVector3Modifier(
     Gameplay=&MakeCustomVectorFromBool(
      Bool=$User.bIsFrontEndPreview
      True Vector=&SelectVectorFromArray(
       Direct Array Index=$Particles.PartitionParticles.Partition
       Vector Selection Array=DataInterface() {propertyValues {FloatData [<0 5 0>, <0 -5 0>]}}
      )
      False Vector=&Lerp_Vector(
       A=&SelectVectorFromArray(
        Direct Array Index=$Particles.PartitionParticles.Partition
        Vector Selection Array=DataInterface() {propertyValues {FloatData [<-1 -2 -1>, <1 2 1>]}}
       )
       B=&SelectVectorFromArray(
        Direct Array Index=$Particles.PartitionParticles.Partition
        Vector Selection Array=DataInterface() {propertyValues {FloatData [<0 0 0>, <0 0 0>]}}
       )
       Alpha=$User.ParaGlideLeanAlpha
      )
     )
     Front End=&Lerp_Vector(
      A=&SelectVectorFromArray(
       Direct Array Index=$Particles.PartitionParticles.Partition
       Vector Selection Array=DataInterface() {propertyValues {FloatData [<0 0 0>, <0 0 0>]}}
      )
      B=&SelectVectorFromArray(
       Direct Array Index=$Particles.PartitionParticles.Partition
       Vector Selection Array=DataInterface() {propertyValues {FloatData [<0 0 0>, <0 0 0>]}}
      )
      Alpha=$User.ParaGlideLeanAlpha
     )
    )
    Skeletal Mesh=$System.Skeletal Mesh
   )
   &UpdateMeshOrientation(
    Facing Coordinate Space:ENiagaraCoordinateSpace=::Local
    Facing Interpolation=false
    Orientation Method=::NewEnumerator1
    Override Quat=false
    Side Coordinate Space:ENiagaraCoordinateSpace=::Local
    Stabilization Mode:ENiagara_OrientVectorCount=::NewEnumerator3
    TransformFacing=true
    TransformSide=true
    TransformUp=true
    Up Coordinate Space:ENiagaraCoordinateSpace=::World
    Up Direction:Vector3f=&FN_FrontEndVector3Modifier(
     Gameplay=&Multiply_Vector(
      A=<-1 -1 -3>
      B=&NormalizeVector(Vector To Normalize=$Engine.Owner.Velocity, Safe Normalize=true, Safe Fallback Threshold=0, Safe Fallback Vector=<0 0 -0.001>)
     )
     Front End=&Multiply_Vector(
      A=<-1 -1 -2>
      B=&TransformVector(
       Vector=&NormalizeVector(
        Vector To Normalize=&Lerp_Vector(A=<0 0.2 2>, B=<0 2 2>, Alpha=$User.ParaGlideLeanAlpha)
        Safe Normalize=true
        Safe Fallback Threshold=0
        Safe Fallback Vector=<0 0 -0.001>
       )
       Destination Space=::World
      )
     )
    )
   )
  ]
  rendererClasses [&NiagaraMeshRendererProperties]
  renderers [
   {rendererClass=&NiagaraMeshRendererProperties, rendererIndex=0}
   {rendererClass=&NiagaraMeshRendererProperties, rendererIndex=1}
  ]
 }
 Sparkles {
  emitterUpdateScript [
   &EmitterState()
   &SpawnRate(SpawnRate=&FN_FrontEndFloatModifier(Gameplay=35, Front End=25), Use Spawn Probability=false)
  ]
  particleSpawnScript [
   &PartitionParticles()
   &InitializeParticle(
    Color=&FNTimeOfDayColor(
     Night Color=<rgba 0.3 0.3 0.12 1>
     Morning Color=<rgba 0.8 0.66 0.24 1>
     Day Color=<rgba 0.9 0.81 0.36 1>
     Evening Color=<rgba 0.8 0.66 0.24 1>
    )
    Lifetime Max=0.7
    Lifetime Min=0.5
    Lifetime Mode=::NewEnumerator1
    Sprite Size Max=&FN_FrontEndVector2Modifier(Front End:Vector2f=<14 14>, Gameplay:Vector2f=<20 20>)
    Sprite Size Min=&FN_FrontEndVector2Modifier(Front End:Vector2f=<11 11>, Gameplay:Vector2f=<15 12>)
    Sprite Size Mode=::NewEnumerator2
    Sprite UV Mode=::NewEnumerator3
    UsePositionOffset=false
    Write Material Random=true
   )
   &SkeletalMeshLocation(
    Bone / Socket Index=$Particles.PartitionParticles.Partition
    Bone Sampling Mode=::NewEnumerator4
    MaskSpawnGroup=false
    Position Sampling=::NewEnumerator1
    Skeletal Mesh=$System.Skeletal Mesh
   ) {module=SkeletalMeshLocation001}
   &ShapeLocation(MaskSpawnGroup=false, Override Seed=false, Sphere Radius=20, Sphere Surface Distribution=1, UseOwnerScale=true)
   &AddVelocity(
    Rotation Coordinate Space=::Simulation
    Velocity=&FN_FrontEndVector3Modifier(
     Gameplay=<0 0 150>
     Front End=&Lerp_Vector(A=<0 0 125>, B=<0 -75 125>, Alpha=&ClampFloat(Float=$User.ParaGlideLeanAlpha))
    )
    Velocity Speed Scale=&Lerp_Float(A=2, B=3, Alpha=&ClampFloat(Float=$User.ParaGlideLeanAlpha))
   )
   &SubUVAnimation(
    End Frame Range Override=3
    "Sprite Renderer "=DataInterface() {propertyValues {SpriteRenderer=&self:Sparkles_0.NiagaraSpriteRendererProperties_0}}
    SubUV Animation Mode:ENiagara_SubUVLookupModeV2=::NewEnumerator2
    UseEndFrame=true
    UseStartFrame=true
   )
   &CameraOffset(Camera Offset Amount=32)
  ]
  particleUpdateScript [
   &ParticleState()
   &ScaleSpriteSize(
    Uniform Curve Scale=&Waveform(
     Global Phase=$Particles.Age
     "Waveform [1]"=::NewEnumerator0
     "[1] Phase Scale"=&FloatFromCurve(FloatCurve=DataInterface() {propertyValues {Curve {keys [<at 0 1>,<at 1 0>]}}})
     "[1] Phase"=$Particles.Age
     "[1] Frequency"=30
     "[1] Amplitude Min/Max"=<0.15 0.9>
    )
    Uniform Curve Sprite Scale=DataInterface() {propertyValues {Curve {keys [
        {interpMode=RCIM_Cubic, leaveTangentWeight=0.4178, tangentMode=RCTM_User, time=0, value=0}
        {interpMode=RCIM_Cubic, tangentMode=RCTM_User, time=0.25, value=1}
        {arriveTangentWeight=0.1389, interpMode=RCIM_Cubic, leaveTangentWeight=0.2778, tangentMode=RCTM_User, time=0.5, value=1}
        {arriveTangent=-2.6251, interpMode=RCIM_Cubic, leaveTangent=-2.6251, tangentMode=RCTM_User, time=1, value=0}
       ]}}}
   )
   /Niagara/Modules/Update/Forces/CurlNoiseForce(Noise Frequency=100, Noise Strength=1000, Pan Noise=true, Pan Noise Field=<0 0 1>)
   &SolveForcesAndVelocity(Clamp Velocity=false, Limit Acceleration=false)
  ]
  rendererClasses [&NiagaraSpriteRendererProperties]
  renderers [{rendererClass=&NiagaraSpriteRendererProperties}]
 }
 Wind_Lines {
  emitterUpdateScript [
   &EmitterState(
    Inactive Response=::NewEnumerator1
    Life Cycle Mode=::NewEnumerator1
    Loop Behavior=::NewEnumerator1
    Loop Duration Mode=::NewEnumerator1
    UseLoopDelay=false
   )
   &SpawnRate(SpawnRate=5, Use Spawn Probability=false)
  ]
  particleSpawnScript [
   &PartitionParticles()
   &InitializeParticle(
    Color Mode=::NewEnumerator0
    Lifetime=&Lerp_Float(
     A=&RandomRangeFloat(Minimum=0.5, Maximum=0.8)
     B=&RandomRangeFloat(Minimum=0.3, Maximum=0.4)
     Alpha=$User.ParaGlideLeanAlpha
    )
    Mesh Scale Mode=::NewEnumerator3
    Mesh Uniform Scale=&FN_FrontEndFloatModifier(
     Gameplay=&RandomRangeFloat(Minimum=0.9, Maximum=1)
     Front End=&RandomRangeFloat(Minimum=0.8, Maximum=0.9)
    )
    UsePositionOffset=false
    Write Material Random=true
   )
   &InitialMeshOrientation(
    Mesh Axis To Orient=&FN_FrontEndVector3Modifier(Gameplay=<1 0 0>, Front End=<-1 0 0>)
    Mesh Orientation Mode=::NewEnumerator2
    Orientation Coordinate Space=::Local
    Rotation=&RandomRangeVector(
     Minimum=&FN_FrontEndVector3Modifier(
      Gameplay=&SelectVectorFromArray(
       Direct Array Index=$Particles.PartitionParticles.Partition
       Vector Selection Array=DataInterface() {propertyValues {FloatData [<0 0.02 0>, <0 -0.02 0>]}}
      )
      Front End=&SelectVectorFromArray(
       Direct Array Index=$Particles.PartitionParticles.Partition
       Vector Selection Array=DataInterface() {propertyValues {FloatData [<0 0.025 0>, <0 -0.025 0>]}}
      )
     )
     Maximum=&FN_FrontEndVector3Modifier(
      Gameplay=&SelectVectorFromArray(
       Direct Array Index=$Particles.PartitionParticles.Partition
       Vector Selection Array=DataInterface() {propertyValues {FloatData [<0 0.02 1>, <0 -0.02 1>]}}
      )
      Front End=&SelectVectorFromArray(
       Direct Array Index=$Particles.PartitionParticles.Partition
       Vector Selection Array=DataInterface() {propertyValues {FloatData [<0 0.025 1>, <0 -0.025 -2>]}}
      )
     )
    )
    Rotation Coordinate Space=::NewEnumerator2
    Use Rotation Vector=true
   )
  ]
  particleUpdateScript [
   &ParticleState()
   &Timeline(
    Play Rate=&Lerp_Float(A=1, B=2, Alpha=&ClampFloat(Float=$User.ParaGlideLeanAlpha))
    Timeline End=10000000
   )
   &DynamicMaterialParameters(
    Index 0 Param 1=&SelectFloatFromArray(Direct Array Index=$Particles.PartitionParticles.Partition, Float Selection Array=DataInterface() {propertyValues {FloatData [0,314]}})
    Index 0 Param 2=$Particles.Timeline.Time
    Index 0 Param 3=$User.ParaGlideLeanAlpha
    Index 0 Param 4=&RandomRangeFloat(Minimum=1, Maximum=2)
    Module.Index0Param0Write=true
    Module.Index0Param1Write=true
    Module.Index0Param2Write=true
    Module.Index0Param3Write=true
   ) {module=DynamicMaterialParameters001}
   &SkeletalMeshLocation(
    Bone / Socket Index=$Particles.PartitionParticles.Partition
    Bone Sampling Mode=::NewEnumerator4
    MaskSpawnGroup=false
    Orientation Sampling=::NewEnumerator1
    Sampled Position Offset=&FN_FrontEndVector3Modifier(
     Gameplay=&MakeCustomVectorFromBool(
      Bool=$User.bIsFrontEndPreview
      True Vector=&SelectVectorFromArray(
       Direct Array Index=$Particles.PartitionParticles.Partition
       Vector Selection Array=DataInterface() {propertyValues {FloatData [<0 5 0>, <0 -5 0>]}}
      )
      False Vector=&SelectVectorFromArray(
       Direct Array Index=$Particles.PartitionParticles.Partition
       Vector Selection Array=DataInterface() {propertyValues {FloatData [<-1 -2 -1>, <1 2 1>]}}
      )
     )
     Front End=&Lerp_Vector(
      A=&SelectVectorFromArray(
       Direct Array Index=$Particles.PartitionParticles.Partition
       Vector Selection Array=DataInterface() {propertyValues {FloatData [<0 0 0>, <0 0 0>]}}
      )
      B=&SelectVectorFromArray(
       Direct Array Index=$Particles.PartitionParticles.Partition
       Vector Selection Array=DataInterface() {propertyValues {FloatData [<0 0 0>, <0 0 0>]}}
      )
      Alpha=$User.ParaGlideLeanAlpha
     )
    )
    Skeletal Mesh=$System.Skeletal Mesh
   )
   &UpdateMeshOrientation(
    Facing Coordinate Space:ENiagaraCoordinateSpace=::Local
    Facing Interpolation=false
    Orientation Method=::NewEnumerator1
    Override Quat=false
    Side Coordinate Space:ENiagaraCoordinateSpace=::Local
    Stabilization Mode:ENiagara_OrientVectorCount=::NewEnumerator3
    TransformFacing=true
    TransformSide=true
    TransformUp=true
    Up Coordinate Space:ENiagaraCoordinateSpace=::World
    Up Direction:Vector3f=&FN_FrontEndVector3Modifier(
     Gameplay=&Multiply_Vector(
      A=<-1 -1 -2>
      B=&NormalizeVector(Vector To Normalize=$Engine.Owner.Velocity, Safe Normalize=true, Safe Fallback Threshold=0, Safe Fallback Vector=<0 0 -0.001>)
     )
     Front End=&Multiply_Vector(
      A=<-1 -1 -2>
      B=&TransformVector(
       Vector=&NormalizeVector(
        Vector To Normalize=&Lerp_Vector(A=<0 0.2 2>, B=<0 2 2>, Alpha=$User.ParaGlideLeanAlpha)
        Safe Normalize=true
        Safe Fallback Threshold=0
        Safe Fallback Vector=<0 0 -0.001>
       )
       Destination Space=::World
      )
     )
    )
   )
  ]
  rendererClasses [&NiagaraMeshRendererProperties]
  renderers [{rendererClass=&NiagaraMeshRendererProperties, rendererIndex=0}]
 }
}
summary {
 systemName=NS_ConTrail_ZebraScramble
 emitters [
  {emitterName=Sparkles, rendererClasses [&NiagaraSpriteRendererProperties]}
  {emitterName=Contrail_Mesh, rendererClasses [&NiagaraMeshRendererProperties]}
  {emitterName=Wind_Lines, rendererClasses [&NiagaraMeshRendererProperties]}
 ]
 userVariables [
  User.bIsFrontEnd {type=NiagaraBool}
  User.bIsFrontEndPreview {type=NiagaraBool}
  User.ParaGlideLeanAlpha {type=NiagaraFloat, defaultValue=0, description=""}
 ]
}
