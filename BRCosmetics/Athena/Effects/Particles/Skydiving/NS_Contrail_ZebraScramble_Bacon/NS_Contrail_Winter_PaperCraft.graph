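# Niagara skydiving contrail for a winter papercraft cosmetic, trailing two chains of mesh particles from skeletal mesh bones or sockets.
# A 14-particle burst splits into two partitions, stacked with increasing scale, alternating rotation and mirrored sine/cosine sway.
# A scalability-distance check drives a timeline that curves Z scale and offset to extend or retract the chains.
# Orientation follows owner velocity, or glider lean in front end, and color tints by time of day.
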NiagaraSystem /BRCosmetics/Athena/Effects/Particles/Skydiving/NS_Contrail_ZebraScramble_Bacon/NS_Contrail_Winter_PaperCraft
dependencies {
 usedDataInterfaces [&NiagaraDataInterfaceCurve, &NiagaraDataInterfaceSkeletalMesh]
 usedDynamicInputs [&FNTimeOfDayColor]
 usedModules [
  &self:SystemSpawnScript.SystemScriptSource.SystemScriptGraph.NiagaraNodeAssignment_1.SetVariables_#1
  &SystemState
 ]
}
emitters {Contrail_Mesh {
  emitterUpdateScript [
   &EmitterState(
    Inactive Response=::NewEnumerator1
    Life Cycle Mode=::NewEnumerator1
    Loop Behavior=::NewEnumerator1
    Loop Duration Mode=::NewEnumerator1
    UseLoopDelay=false
   )
   /SetVariables(Emitter.DistanceFractionThreshold:NiagaraBool=&SetBoolByFloatComparison(A=$Emitter.Scalability.DistanceFraction, B=0.9, Comparison Type=::NewEnumerator6)) {isSetParameters=true}
   &SpawnBurst_Instantaneous(Spawn Count=14, Use Spawn Probability=false, UseLoopCountLimit=false)
   &Timeline(Play=$Emitter.DistanceFractionThreshold, Play Rate=4, Rewind Rate=4) {module=Timeline001}
  ]
  particleSpawnScript [
   &PartitionParticles(Particles Per Partition=7, Partition Mode:ENiagara_PartitionMode_v2=::NewEnumerator1)
   &InitializeParticle(
    Color Mode=::NewEnumerator0
    Mesh Scale Mode=::NewEnumerator3
    Mesh Uniform Scale=&Add_Float(
     A=1
     B=&Multiply_Float(A=0.05, B=&Multiply_Float(A=$Particles.PartitionParticles.Normalized Index In Partition, B=7))
    )
    UsePositionOffset=false
    Write Material Random=true
   )
   &InitialMeshOrientation(
    Mesh Axis To Orient=&FN_FrontEndVector3Modifier(Gameplay=<1 0 0>, Front End=<-1 0 0>)
    Mesh Orientation Mode=::NewEnumerator2
    Orientation Coordinate Space=::Local
    Rotation=&MakeVector(
     X=0
     Y=0
     Z=&MaskFloat(
      Mask Source=0
      Mask Parameter=?
      Mask=&ModuloInt(A=$Particles.PartitionParticles.Index In Partition, B=2)
      Masked Float=0
      Passthrough Float=0.25
     )
    )
    Rotation Coordinate Space=::NewEnumerator2
    Use Rotation Vector=true
   )
   /SetVariables(Particles.OffsetPosition:Vector3f=&MakeVector(
     X=0
     Y=0
     Z=&Add_Float(
      A=1
      B=&Multiply_Float(
       A=&Multiply_Float(A=20, B=&MakeFloatFromVector(Vector=$Particles.Scale, Channel=::NewEnumerator0))
       B=&Multiply_Float(A=$Particles.PartitionParticles.Normalized Index In Partition, B=6)
      )
     )
    )) {isSetParameters=true}
  ]
  particleUpdateScript [
   &ParticleState(Kill Particles When Lifetime Has Elapsed=false, Let Infinitely Lived Particles Die When Emitter Deactivates=true)
   &ScaleMeshSize(Scale Factor=&MakeVector(
     X=1
     Y=1
     Z=&FloatFromCurve(
      FloatCurve=DataInterface() {propertyValues {Curve {keys [
          {arriveTangent=15.9072, interpMode=RCIM_Cubic, leaveTangent=15.9072, tangentMode=RCTM_User, time=0, value=0}
          {arriveTangent=1.2036, interpMode=RCIM_Cubic, leaveTangent=1.2036, tangentMode=RCTM_User, time=0.45, value=1}
          <at 1 1>
         ]}}}
      CurveIndex=$Emitter.Age
      Scale Curve=$Emitter.Timeline001.NormalizedTime
     )
    ))
   &Color(Color=&FN_Select_Color_from_Time_Of_Day(Morning=<rgba 0.7 0.6475 0.595 1>, Day=<rgba 1 1 1 1>, Evening=<rgba 0.8 0.74 0.68 1>, Night=<rgba 0.3 0.3 0.5 1>))
   &SkeletalMeshLocation(
    Bone / Socket Index=$Particles.PartitionParticles.Partition
    Bone Sampling Mode=::NewEnumerator4
    MaskSpawnGroup=false
    Orientation Sampling=::NewEnumerator1
    Sampled Position Offset=&Lerp_Vector(
     A=<0 0 0>
     B=&Add_Vector(
      A=$Particles.OffsetPosition
      B=&MakeVector(
       X=&Cosine(
        Normalized Angle=$Emitter.Age
        Period=1
        Scale=&Multiply_Float(
         A=$Particles.Initial.PartitionParticles.Normalized Index In Partition
         B=&MaskFloat(Mask Source=0, Mask Parameter=?, Mask=$Particles.PartitionParticles.Partition, Masked Float=-10, Passthrough Float=10)
        )
        Bias=0
       )
       Y=&Sine(
        Normalized Angle=$Emitter.Age
        Period=1
        Scale=&Multiply_Float(
         A=$Particles.PartitionParticles.Normalized Index In Partition
         B=&MaskFloat(Mask Source=0, Mask Parameter=?, Mask=$Particles.PartitionParticles.Partition, Masked Float=10, Passthrough Float=-10)
        )
        Bias=0
       )
       Z=0
      )
     )
     Alpha=&FloatFromCurve(
      FloatCurve=DataInterface() {propertyValues {Curve {keys [
          {arriveTangent=15.9072, interpMode=RCIM_Cubic, leaveTangent=15.9072, tangentMode=RCTM_User, time=0, value=0}
          {arriveTangent=1.2036, interpMode=RCIM_Cubic, leaveTangent=1.2036, tangentMode=RCTM_User, time=0.45, value=1}
          <at 1 1>
         ]}}}
      CurveIndex=$Emitter.Age
      Scale Curve=$Emitter.Timeline001.NormalizedTime
     )
    )
    Sampled Position Offset Coordinate Space:ENiagaraSkelMeshTransforms=::NewEnumerator2
    Skeletal Mesh=$System.Skeletal Mesh
   )
   &UpdateMeshOrientation(
    Facing Coordinate Space:ENiagaraCoordinateSpace=::Local
    Facing Interpolation=false
    Orientation Method=::NewEnumerator1
    Override Quat=false
    Side Coordinate Space:ENiagaraCoordinateSpace=::Local
    Stabilization Mode:ENiagara_OrientVectorCount=::NewEnumerator3
    TransformFacing=true
    TransformSide=true
    TransformUp=true
    Up Coordinate Space:ENiagaraCoordinateSpace=::World
    Up Direction:Vector3f=&FN_FrontEndVector3Modifier(
     Gameplay=&Multiply_Vector(
      A=<-1 -1 -3>
      B=&NormalizeVector(Vector To Normalize=$Engine.Owner.Velocity, Safe Normalize=true, Safe Fallback Threshold=0, Safe Fallback Vector=<0 0 -0.001>)
     )
     Front End=&Multiply_Vector(
      A=<-1 -1 -2>
      B=&TransformVector(
       Vector=&NormalizeVector(
        Vector To Normalize=&Lerp_Vector(A=<0 0.2 2>, B=<0 2 2>, Alpha=$User.ParaGlideLeanAlpha)
        Safe Normalize=true
        Safe Fallback Threshold=0
        Safe Fallback Vector=<0 0 -0.001>
       )
       Destination Space=::World
      )
     )
    )
   )
  ]
  rendererClasses [&NiagaraMeshRendererProperties]
  renderers [{rendererClass=&NiagaraMeshRendererProperties, rendererIndex=0}]
 }}
summary {
 systemName=NS_Contrail_Winter_PaperCraft
 emitters [{emitterName=Contrail_Mesh, rendererClasses [&NiagaraMeshRendererProperties]}]
 userVariables [
  User.bIsFrontEnd {type=NiagaraBool}
  User.bIsFrontEndPreview {type=NiagaraBool}
  User.ParaGlideLeanAlpha {type=NiagaraFloat, defaultValue=0, description=""}
 ]
}
